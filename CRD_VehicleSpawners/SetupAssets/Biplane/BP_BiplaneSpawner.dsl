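; Creative biplane spawner that applies user options to its spawned Ferret plane, replacing the normal valet-vehicle options flow.
; Option changes, property loads and vehicle spawns trigger an authority-only update of prop-destruction thresholds, glider redeploy and explosion damage, writing only changed values to limit replication.
; UpdateVehicleConfigs also fills in missing impulse-response physics and sets explosion delay and lifetime.
; Option events also set the PreGame enabled state, behind a preview check noted as broken.
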(event EventBeginPlay
  (|CallParentFunction))

(event OnAnyPropertyChanged(ToyOptionsComponent)
  ;[1 0.239 0.2291] Override normal options flow for plane since it uses a specific set vs. valet vehicles.
  ;;[0 0 0 1 14] This previewing check currently does not work. It should be fixed shortly, so this will work when that happens.
  (if (not (Building|IsPreviewBuildingActor))
    (Class|CreativeEnabledComponent|SetEnabledOnPhase (Variables|Default|GetEnabledComponent) "PreGame" (Variables|LegacyOptions|GetEnabledIndex))
    (Default|UpdatePlanewithNewThresholds)))

(event OnPropertiesLoaded(ToyOptionsComponent)
  ;[1 0.239 0.2291] Override normal options flow for plane since it uses a specific set vs. valet vehicles.
  ;;[0 0 0 1 14] This previewing check currently does not work. It should be fixed shortly, so this will work when that happens.
  (if (not (Building|IsPreviewBuildingActor))
    (Class|CreativeEnabledComponent|SetEnabledOnPhase (Variables|Default|GetEnabledComponent) "PreGame" (Variables|LegacyOptions|GetEnabledIndex))
    (Default|UpdatePlanewithNewThresholds)))

(event EventOnVehicleSpawned
  (params (Spawned%20Vehicle :type Fort%20Athena%20Vehicle%20Object%20Reference))
  ;[1 0.5785 0.9298] When spawning, make sure we update relevant options
  (|CallParentFunction Spawned%20Vehicle)
  (Default|UpdatePlanewithNewThresholds))

(fn SetImpulseResponsePhysicsIfNone
  ;[0.2031 0.343 1 0.4] If the vehicle is not set to respond to impulse physics, set it to do so with base values.
  (bind _asfort_athena_vehicle (Utilities|Casting|CastToFortAthenaVehicle (Variables|Default|GetCachedVehicleInterface)))
  (if (<= (Variables|/Script/FortniteGame.FortAthenaVehicle|GetImpulseResponseMultiplier _asfort_athena_vehicle) 0)
    (Class|FortAthenaVehicle|SetImpulseResponseMultiplier _asfort_athena_vehicle (Variables|Default|GetBaseImpulseResponseMultiplier))
    (else
      _))
  (if (<= (Variables|/Script/FortniteGame.FortAthenaVehicle|GetImpulseResponseZBias _asfort_athena_vehicle) 0)
    (Class|FortAthenaVehicle|SetImpulseResponseZBias _asfort_athena_vehicle (Variables|Default|GetBaseZBiasImpulseResponse))))

(fn UpdateVehicleConfigs
  (|CallParentFunction)
  (CallFunction|SetImpulseResponsePhysicsifNone)
  (CallFunction|UpdateExplosionLifetimeSettings))

(fn UpdateExplosionLifetimeSettings
  (bind _asfort_athena_vehicle (Utilities|Casting|CastToFortAthenaVehicle (Variables|Default|GetCachedVehicleInterface)))
  (bind _returnvalue (Class|FortAthenaVehicle|GetCurrentRuntimeProperties _asfort_athena_vehicle))
  (bind (_returnvalue_bhasinfinitefuel _returnvalue_bsupportcosmeticwrap _returnvalue_waterdestructiondelay _returnvalue_selfdestructtime _returnvalue_maxtimeexplosion _returnvalue_maxtimezerohealth _returnvalue_maxtimeselfdestructcleanup) (Utilities|Struct|BreakVehicleRuntimeModifiers _returnvalue))
  (Class|FortAthenaVehicle|ServerSetDestructionInfo _asfort_athena_vehicle _returnvalue_selfdestructtime (Variables|UserOptions|Advanced|GetLifetimeAfterExplosion) _returnvalue_maxtimeselfdestructcleanup (Variables|UserOptions|Advanced|GetMaxExplosionDelay)))

(fn UpdatePlaneWithNewThresholds
  (bind _boostdamagehealththreshold (Variables|Default|GetBoostDamageHealthThreshold))
  (bind _allowgliderredeploy (Variables|UserOptions|Advanced|GetAllowGliderRedeploy))
  (bind _damagehealththreshold (Variables|Default|GetDamageHealthThreshold))
  (bind _directdamagehealththreshold (Variables|Default|GetDirectDamageHealthThreshold))
  (Networking|SwitchHasAuthority
    (:Authority
      (if (Utilities|IsValid (Utilities|ToObject(Interface) (Variables|Default|GetCachedVehicleInterface)) :!class /Script/Engine.KismetSystemLibrary)
        (bind _asferret_vehicle (Utilities|Casting|CastToFerretVehicle (Utilities|ToObject(Interface) (Variables|Default|GetCachedVehicleInterface))))
        (if (Utilities|IsValid _asferret_vehicle :!class /Script/Engine.KismetSystemLibrary)
          (bind _output_get (LocalVariables|SetFerretPlane _asferret_vehicle))
          (Variables|/Game/Athena/DrivableVehicles/Biplane/Blueprints/FerretVehicle.FerretVehicle_C|SetbValetCurieEnabled true _output_get)
          (Utilities|FlowControl|Sequence
            (:then_0
              ;[1 1 1] Only set new values if they are different from what is already set.  Prevents data from replicating that doesn't need to
              (if (!= (Variables|/Script/FortniteGame.FortAthenaFerretVehicle|GetMaxHealthToDestroyPropWithDirectHit (LocalVariables|GetFerretPlane)) _directdamagehealththreshold)
                (Variables|/Script/FortniteGame.FortAthenaFerretVehicle|SetMaxHealthToDestroyPropWithDirectHit _directdamagehealththreshold (LocalVariables|GetFerretPlane))))
            (:then_1
              (if (!= (Variables|/Script/FortniteGame.FortAthenaFerretVehicle|GetMaxHealthToDestroyPropWhileBoosting (LocalVariables|GetFerretPlane)) _boostdamagehealththreshold)
                (Variables|/Script/FortniteGame.FortAthenaFerretVehicle|SetMaxHealthToDestroyPropWhileBoosting _boostdamagehealththreshold (LocalVariables|GetFerretPlane))))
            (:then_2
              (if (!= (Variables|/Script/FortniteGame.FortAthenaFerretVehicle|GetMaxHealthToDestroyProp (LocalVariables|GetFerretPlane)) _damagehealththreshold)
                (Variables|/Script/FortniteGame.FortAthenaFerretVehicle|SetMaxHealthToDestroyProp _damagehealththreshold (LocalVariables|GetFerretPlane))))
            (:then_3
              (if (!= (Variables|/Game/Athena/DrivableVehicles/Biplane/Blueprints/FerretVehicle.FerretVehicle_C|GetGliderRedeployOnExit (LocalVariables|GetFerretPlane)) _allowgliderredeploy)
                (Variables|/Game/Athena/DrivableVehicles/Biplane/Blueprints/FerretVehicle.FerretVehicle_C|SetGliderRedeployOnExit _allowgliderredeploy (LocalVariables|GetFerretPlane))))
            (:then_4
              (if (!= (Variables|/Game/Athena/DrivableVehicles/Biplane/Blueprints/FerretVehicle.FerretVehicle_C|GetExplosionDamageBuildings (LocalVariables|GetFerretPlane)) (Variables|UserOptions|Advanced|GetExplosionDamageToEnvironment))
                (Variables|/Game/Athena/DrivableVehicles/Biplane/Blueprints/FerretVehicle.FerretVehicle_C|SetExplosionDamageBuildings (Variables|UserOptions|Advanced|GetExplosionDamageToEnvironment) (LocalVariables|GetFerretPlane))))
            (:then_5
              (if (!= (Variables|/Game/Athena/DrivableVehicles/Biplane/Blueprints/FerretVehicle.FerretVehicle_C|GetExplosionDamagePlayers (LocalVariables|GetFerretPlane)) (Variables|UserOptions|Advanced|GetExplosionDamageToPlayers))
                (Variables|/Game/Athena/DrivableVehicles/Biplane/Blueprints/FerretVehicle.FerretVehicle_C|SetExplosionDamagePlayers (Variables|UserOptions|Advanced|GetExplosionDamageToPlayers) (LocalVariables|GetFerretPlane))))
            (:then_6
              (if (!= (Variables|/Game/Athena/DrivableVehicles/Biplane/Blueprints/FerretVehicle.FerretVehicle_C|GetExplosionDamageVehicles (LocalVariables|GetFerretPlane)) (Variables|UserOptions|Advanced|GetExplosionDamageToVehicles))
                (Variables|/Game/Athena/DrivableVehicles/Biplane/Blueprints/FerretVehicle.FerretVehicle_C|SetExplosionDamageVehicles (Variables|UserOptions|Advanced|GetExplosionDamageToVehicles) (LocalVariables|GetFerretPlane))))))))))

(fn ConstructionScript
  (|CallParentFunction))
