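; Creative biplane spawner device that applies its user options to the spawned plane (FerretVehicle).
; Option changes, property loading and spawning call UpdatePlaneWithNewThresholds, which on authority writes prop-destruction health thresholds, glider redeploy and explosion damage, only where values differ to limit replication.
; UpdateVehicleConfigs fills in missing impulse response and sets destruction timing.
; Fuel and wrap accessors map to user options, and migration converts legacy thresholds and wrap support.
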(event EventBeginPlay
  (|CallParentFunction))

(event OnAnyPropertyChanged(ToyOptionsComponent)
  ;[1 0.239 0.2291] Override normal options flow for plane since it uses a specific set vs. valet vehicles.
  ;;[0 0 0 1 14] This previewing check currently does not work. It should be fixed shortly, so this will work when that happens.
  (if (not (Building|IsPreviewBuildingActor))
    (Class|CreativeEnabledComponent|SetEnabledOnPhase (Variables|Default|GetEnabledComponent) "PreGame" (Variables|LegacyOptions|GetEnabledIndex))
    (Default|UpdatePlanewithNewThresholds)))

(event OnPropertiesLoaded(ToyOptionsComponent)
  ;[1 0.239 0.2291] Override normal options flow for plane since it uses a specific set vs. valet vehicles.
  ;;[0 0 0 1 14] This previewing check currently does not work. It should be fixed shortly, so this will work when that happens.
  (if (not (Building|IsPreviewBuildingActor))
    (Class|CreativeEnabledComponent|SetEnabledOnPhase (Variables|Default|GetEnabledComponent) "PreGame" (Variables|LegacyOptions|GetEnabledIndex))
    (Default|UpdatePlanewithNewThresholds)))

(event EventOnVehicleSpawned
  (params (Spawned%20Vehicle :type Fort%20Athena%20Vehicle%20Object%20Reference))
  ;[1 0.5785 0.9298] When spawning, make sure we update relevant options
  (|CallParentFunction Spawned%20Vehicle)
  (Default|UpdatePlanewithNewThresholds)
  (Variables|/Script/FortniteGame.FortAthenaVehicle|SetbSkydivePlayerOnVehicleExit (Variables|UserOptions|Advanced|GetSkydive%20on%20Exit) Spawned%20Vehicle))

(fn GetSupportsWraps
  (return (Variables|UserOptions|Basic|GetSupportsWraps)))

(fn GetConsumesFuel
  (return (Variables|UserOptions|GetFuel%20Consumption)))

(fn GetFuelUseMultiplier
  (return (Variables|UserOptions|GetFuel%20Use)))

(fn GetMaxRandomStartingFuel
  (return (Variables|UserOptions|Advanced|GetMax%20Random%20Starting%20Fuel)))

(fn GetMinRandomStartingFuel
  (return (Variables|UserOptions|Advanced|GetMin%20Random%20Starting%20Fuel)))

(fn GetRandomStartingFuel
  (return (Variables|UserOptions|GetRandom%20Starting%20Fuel)))

(fn GetStartingFuelValue
  (return (Variables|UserOptions|GetStarting%20Fuel)))

(fn MigrateLegacyUserOptions
  (bind _bvehiclesupportswraps (Class|GetClassDefaults "/CRD_VehicleSpawners/SetupAssets/Biplane/Device_BiplaneSpawner_V2.Device_BiplaneSpawner_V2_C"))
  (bind _boostdamagehealththreshold (Variables|Legacy|GetBoostDamageHealthThreshold))
  (bind _directdamagehealththreshold (Variables|Legacy|GetDirectDamageHealthThreshold))
  (bind _damagehealththreshold (Variables|Legacy|GetDamageHealthThreshold))
  (Utilities|FlowControl|Sequence
    (:then_0
      (Utilities|FlowControl|Sequence
        (:then_0
          (if (>= _boostdamagehealththreshold 0)
            (Variables|UserOptions|SetOverride%20Battle%20Royale%20Boosted%20Health%20Threshold true)
            (Variables|UserOptions|SetBoostedHealthThreshold _boostdamagehealththreshold)
            (Variables|Legacy|SetBoostDamageHealthThreshold -1.0)
            (LocalVariables|SetOptionMigrated true)))
        (:then_1
          (if (>= _directdamagehealththreshold 0)
            (Variables|UserOptions|SetOverride%20Battle%20Royale%20Direct%20Hit%20Health%20Threshold true)
            (Variables|UserOptions|SetDirectHitHealthThreshold _directdamagehealththreshold)
            (Variables|Legacy|SetDirectDamageHealthThreshold -1.0)
            (LocalVariables|SetOptionMigrated true)))
        (:then_2
          (if (>= _damagehealththreshold 0)
            (Variables|UserOptions|SetOverride%20Battle%20Royale%20Breakthrough%20Health%20Threshold true)
            (Variables|UserOptions|SetBreakthroughHealthThreshold _damagehealththreshold)
            (Variables|Legacy|SetDamageHealthThreshold -1.0)
            (LocalVariables|SetOptionMigrated true)))
        (:then_3
          (if (!= (Variables|Legacy|GetbVehicleSupportsWraps) _bvehiclesupportswraps)
            (Variables|UserOptions|Basic|SetSupportsWraps false)
            (Variables|Legacy|SetbVehicleSupportsWraps _bvehiclesupportswraps)
            (LocalVariables|SetOptionMigrated true)))))
    (:then_1
      (if (LocalVariables|GetOptionMigrated)
        (OptionData|RefreshPropertyOverrides (Variables|Default|GetToyOptionsComponent))))
    (:then_2
      (|CallParentFunction))))

(fn SetConsumeFuel
  (params (ConsumeFuel :type Boolean))
  (Variables|UserOptions|SetFuel%20Consumption ConsumeFuel))

(fn SetStartingFuelValue
  (params (Starting%20Fuel%20Value :type Integer))
  (Variables|UserOptions|SetStarting%20Fuel Starting%20Fuel%20Value))

(fn SetImpulseResponsePhysicsIfNone
  ;[0.2031 0.343 1 0.4] If the vehicle is not set to respond to impulse physics, set it to do so with base values.
  (bind _asfort_athena_vehicle (Utilities|Casting|CastToFortAthenaVehicle (Variables|Default|GetCachedVehicleInterface)))
  (if (<= (Variables|/Script/FortniteGame.FortAthenaVehicle|GetImpulseResponseMultiplier _asfort_athena_vehicle) 0)
    (Class|FortAthenaVehicle|SetImpulseResponseMultiplier _asfort_athena_vehicle (Variables|Default|GetBaseImpulseResponseMultiplier))
    (else
      _))
  (if (<= (Variables|/Script/FortniteGame.FortAthenaVehicle|GetImpulseResponseZBias _asfort_athena_vehicle) 0)
    (Class|FortAthenaVehicle|SetImpulseResponseZBias _asfort_athena_vehicle (Variables|Default|GetBaseZBiasImpulseResponse))))

(fn SetMaxRandomStartingFuel
  (params (Max%20Fuel%20Percent :type Integer))
  (|CallParentFunction Max%20Fuel%20Percent)
  (Variables|UserOptions|Advanced|SetMax%20Random%20Starting%20Fuel Max%20Fuel%20Percent))

(fn SetMinRandomStartingFuel
  (params (Min%20Fuel%20Percent :type Integer))
  (|CallParentFunction Min%20Fuel%20Percent)
  (Variables|UserOptions|Advanced|SetMin%20Random%20Starting%20Fuel Min%20Fuel%20Percent))

(fn UpdateVehicleConfigs
  (|CallParentFunction)
  (CallFunction|SetImpulseResponsePhysicsifNone)
  (CallFunction|UpdateExplosionLifetimeSettings))

(fn UpdateExplosionLifetimeSettings
  (bind _asfort_athena_vehicle (Utilities|Casting|CastToFortAthenaVehicle (Variables|Default|GetCachedVehicleInterface)))
  (bind _returnvalue (Class|FortAthenaVehicle|GetCurrentRuntimeProperties _asfort_athena_vehicle))
  (bind (_returnvalue_bhasinfinitefuel _returnvalue_bsupportcosmeticwrap _returnvalue_waterdestructiondelay _returnvalue_selfdestructtime _returnvalue_maxtimeexplosion _returnvalue_maxtimezerohealth _returnvalue_maxtimeselfdestructcleanup) (Utilities|Struct|BreakVehicleRuntimeModifiers _returnvalue))
  (Class|FortAthenaVehicle|ServerSetDestructionInfo _asfort_athena_vehicle _returnvalue_selfdestructtime (Variables|UserOptions|Advanced|GetLifetimeAfterExplosion) _returnvalue_maxtimeselfdestructcleanup (Variables|UserOptions|Advanced|GetMaxExplosionDelay)))

(fn UpdatePlaneWithNewThresholds
  (bind _override_battle_royale_direct_hit_health_threshold (Variables|UserOptions|GetOverride%20Battle%20Royale%20Direct%20Hit%20Health%20Threshold))
  (bind _directhithealththreshold (Variables|UserOptions|GetDirectHitHealthThreshold))
  (bind _returnvalue (select _override_battle_royale_direct_hit_health_threshold _directhithealththreshold -1.0))
  (bind _override_battle_royale_breakthrough_health_threshold (Variables|UserOptions|GetOverride%20Battle%20Royale%20Breakthrough%20Health%20Threshold))
  (bind _breakthroughhealththreshold (Variables|UserOptions|GetBreakthroughHealthThreshold))
  (bind _returnvalue_2 (select _override_battle_royale_breakthrough_health_threshold _breakthroughhealththreshold -1.0))
  (bind _override_battle_royale_boosted_health_threshold (Variables|UserOptions|GetOverride%20Battle%20Royale%20Boosted%20Health%20Threshold))
  (bind _boostedhealththreshold (Variables|UserOptions|GetBoostedHealthThreshold))
  (bind _returnvalue_3 (select _override_battle_royale_boosted_health_threshold _boostedhealththreshold -1.0))
  (bind _allowgliderredeploy (Variables|UserOptions|Advanced|GetAllowGliderRedeploy))
  (Networking|SwitchHasAuthority
    (:Authority
      (if (Utilities|IsValid (Utilities|ToObject(Interface) (Variables|Default|GetCachedVehicleInterface)) :!class /Script/Engine.KismetSystemLibrary)
        (bind _asferret_vehicle (Utilities|Casting|CastToFerretVehicle (Utilities|ToObject(Interface) (Variables|Default|GetCachedVehicleInterface))))
        (if (Utilities|IsValid _asferret_vehicle :!class /Script/Engine.KismetSystemLibrary)
          (LocalVariables|SetFerretPlane _asferret_vehicle)
          (Utilities|FlowControl|Sequence
            (:then_0
              ;[1 1 1] Only set new values if they are different from what is already set.  Prevents data from replicating that doesn't need to
              (if (!= (Variables|/Script/FortniteGame.FortAthenaFerretVehicle|GetMaxHealthToDestroyPropWithDirectHit (LocalVariables|GetFerretPlane)) _returnvalue)
                (Variables|/Script/FortniteGame.FortAthenaFerretVehicle|SetMaxHealthToDestroyPropWithDirectHit _returnvalue (LocalVariables|GetFerretPlane))))
            (:then_1
              (if (!= (Variables|/Script/FortniteGame.FortAthenaFerretVehicle|GetMaxHealthToDestroyPropWhileBoosting (LocalVariables|GetFerretPlane)) _returnvalue_3)
                (Variables|/Script/FortniteGame.FortAthenaFerretVehicle|SetMaxHealthToDestroyPropWhileBoosting _returnvalue_3 (LocalVariables|GetFerretPlane))))
            (:then_2
              (if (!= (Variables|/Script/FortniteGame.FortAthenaFerretVehicle|GetMaxHealthToDestroyProp (LocalVariables|GetFerretPlane)) _returnvalue_2)
                (Variables|/Script/FortniteGame.FortAthenaFerretVehicle|SetMaxHealthToDestroyProp _returnvalue_2 (LocalVariables|GetFerretPlane))))
            (:then_3
              (if (!= (Variables|/Game/Athena/DrivableVehicles/Biplane/Blueprints/FerretVehicle.FerretVehicle_C|GetGliderRedeployOnExit (LocalVariables|GetFerretPlane)) _allowgliderredeploy)
                (Variables|/Game/Athena/DrivableVehicles/Biplane/Blueprints/FerretVehicle.FerretVehicle_C|SetGliderRedeployOnExit _allowgliderredeploy (LocalVariables|GetFerretPlane))))
            (:then_4
              (if (!= (Variables|/Game/Athena/DrivableVehicles/Biplane/Blueprints/FerretVehicle.FerretVehicle_C|GetExplosionDamageBuildings (LocalVariables|GetFerretPlane)) (Variables|UserOptions|Advanced|GetExplosionDamageToEnvironment))
                (Variables|/Game/Athena/DrivableVehicles/Biplane/Blueprints/FerretVehicle.FerretVehicle_C|SetExplosionDamageBuildings (Variables|UserOptions|Advanced|GetExplosionDamageToEnvironment) (LocalVariables|GetFerretPlane))))
            (:then_5
              (if (!= (Variables|/Game/Athena/DrivableVehicles/Biplane/Blueprints/FerretVehicle.FerretVehicle_C|GetExplosionDamagePlayers (LocalVariables|GetFerretPlane)) (Variables|UserOptions|Advanced|GetExplosionDamageToPlayers))
                (Variables|/Game/Athena/DrivableVehicles/Biplane/Blueprints/FerretVehicle.FerretVehicle_C|SetExplosionDamagePlayers (Variables|UserOptions|Advanced|GetExplosionDamageToPlayers) (LocalVariables|GetFerretPlane))))
            (:then_6
              (if (!= (Variables|/Game/Athena/DrivableVehicles/Biplane/Blueprints/FerretVehicle.FerretVehicle_C|GetExplosionDamageVehicles (LocalVariables|GetFerretPlane)) (Variables|UserOptions|Advanced|GetExplosionDamageToVehicles))
                (Variables|/Game/Athena/DrivableVehicles/Biplane/Blueprints/FerretVehicle.FerretVehicle_C|SetExplosionDamageVehicles (Variables|UserOptions|Advanced|GetExplosionDamageToVehicles) (LocalVariables|GetFerretPlane))))))))))

(fn ConstructionScript
  (|CallParentFunction))
